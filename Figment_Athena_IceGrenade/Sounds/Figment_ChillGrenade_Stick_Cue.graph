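# Sound cue for the Figment Chill Grenade sticking, layering the impact with a warning sound.
# A mixer combines a randomly chosen stick variant, with volume modulation, and a single warning wave that starts after a short fixed delay.
# Each layer overrides a medium-explosion attenuation preset with a custom falloff curve and low-pass filtering, giving the warning a longer range.
# Concurrency is overridden to stop the oldest instance.
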SoundCue /Figment_Athena_IceGrenade/Sounds/Figment_ChillGrenade_Stick_Cue
root=Mixer_1
nodes [
 {
  id=Attenuation_0
  children [Modulator_0]
  properties {
   attenuationSettings=/Game/Sounds/Attenuation/Fort_Explosion_Medium_Athena_Att
   bOverrideAttenuation=true
   attenuationOverrides {
    bAttenuateWithLPF=true
    distanceAlgorithm=Custom
    falloffDistance=5500
    hPFFrequencyAtMax=50
    lPFFrequencyAtMin=500
    lPFRadiusMin=50
    reverbDistanceMin=200
    reverbWetLevelMax=1
    attenuationShapeExtents {x=200}
    customAttenuationCurve {editorCurveData {keys [
       {interpMode=RCIM_Cubic, tangentMode=RCTM_User, time=0.05, value=1}
       {arriveTangent=-0.5291, interpMode=RCIM_Cubic, leaveTangent=-0.5291, tangentMode=RCTM_User, time=0.2069, value=0.151}
       <at 1 0>
      ]}}
   }
  }
 }
 {
  id=Attenuation_2
  children [Delay_1]
  properties {
   attenuationSettings=&Fort_Explosion_Medium_Athena_Att
   bOverrideAttenuation=true
   attenuationOverrides {
    bAttenuateWithLPF=true
    distanceAlgorithm=Custom
    falloffDistance=9500
    hPFFrequencyAtMax=50
    lPFFrequencyAtMin=500
    lPFRadiusMax=5000
    lPFRadiusMin=50
    reverbDistanceMax=15000
    reverbDistanceMin=200
    reverbWetLevelMax=1
    reverbWetLevelMin=0.35
    attenuationShapeExtents {x=200}
    customAttenuationCurve {editorCurveData {keys [
       {arriveTangent=-3, interpMode=RCIM_Cubic, leaveTangent=-3, tangentMode=RCTM_User, time=0.05, value=1}
       {arriveTangent=-0.2738, interpMode=RCIM_Cubic, leaveTangent=-0.2738, tangentMode=RCTM_User, time=0.25, value=0.0468}
       <at 1 0>
      ]}}
   }
  }
 }
 {id=Delay_1, children [WavePlayer_1], properties {delayMax=0.2, delayMin=0.2}}
 {id=Mixer_1, children [Attenuation_0, Attenuation_2], properties {inputVolume [1,1]}}
 {id=Modulator_0, children [Random_1], properties {volumeMax=1.1}}
 {id=Random_1, children [WavePlayer_8, WavePlayer_9, WavePlayer_10], properties {weights [1,1,1]}}
 {id=WavePlayer_1, properties {soundWaveAssetPtr=&/ChillGrenade_Warning_01}}
 {id=WavePlayer_10, properties {soundWaveAssetPtr=&/ChillGrenade_Stick_03}}
 {id=WavePlayer_8, properties {soundWaveAssetPtr=&/ChillGrenade_Stick_01}}
 {id=WavePlayer_9, properties {soundWaveAssetPtr=&/ChillGrenade_Stick_02}}
]
properties {
 bOverrideConcurrency=true
 maxDistance=9700
 soundClassObject=&Fort_WeaponRanged_Fire_3P_PSM
 volumeMultiplier=0.8
 concurrencyOverrides {resolutionRule=StopOldest}
}
